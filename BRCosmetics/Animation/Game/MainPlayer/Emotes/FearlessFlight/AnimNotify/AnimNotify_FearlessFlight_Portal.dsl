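; Anim notify state for the Fearless Flight emote that keeps portal material parameters aligned with the character's portal socket while the notify is active.
; Each tick it reads the world transform of SOCKET_Portal on the skeletal mesh and converts it to a matrix.
; It then pushes the socket location and the three unit axes into the mesh's materials as PortalPosition and PortalAxisX, Y and Z vector parameters.
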(fn Received_NotifyTick
  (params (MeshComp :type Skeletal%20Mesh%20Component%20Object%20Reference) (Animation :type Anim%20Sequence%20Base%20Object%20Reference) (FrameDeltaTime :type Float%20%28single-precision%29) (EventReference :type Anim%20Notify%20Event%20Reference%20Structure%20%28by%20ref%29))
  (bind _returnvalue (Transformation|GetSocketTransform MeshComp "SOCKET_Portal" "RTS_World"))
  (bind _returnvalue_2 (Math|Transform|ToMatrix(Transform) _returnvalue))
  (Rendering|Material|SetVectorParameterValueonMaterials MeshComp "PortalPosition" (Transformation|GetSocketLocation MeshComp "SOCKET_Portal"))
  (Rendering|Material|SetVectorParameterValueonMaterials MeshComp "PortalAxisX" (Math|Matrix|GetUnitAxis(Matrix) _returnvalue_2 "X"))
  (Rendering|Material|SetVectorParameterValueonMaterials MeshComp "PortalAxisY" (Math|Matrix|GetUnitAxis(Matrix) _returnvalue_2 "Y"))
  (Rendering|Material|SetVectorParameterValueonMaterials MeshComp "PortalAxisZ" (Math|Matrix|GetUnitAxis(Matrix) _returnvalue_2 "Z"))
  (return false))
