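# Niagara system for a brief orange-red flash in the Fearless Flight emote, built from two burst emitters.
# GlowFlash spawns a short-lived, growing sprite with two renderers, also computing a brightened FlareColor and a wider-than-tall FlareSize that shrinks over life.
# Volume spawns a longer-lived red mesh whose alpha rises quickly, then fades out.
# Opposite camera offsets separate the glow and volume in depth.
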NiagaraSystem /BRCosmetics/Animation/Game/MainPlayer/Emotes/FearlessFlight/FX/NS_Emote_FearlessFlight_CloseFlash
dependencies {usedDataInterfaces [&NiagaraDataInterfaceCurve], usedDynamicInputs [], usedModules [&SystemState]}
emitters {
 GlowFlash {
  emitterUpdateScript [
   &EmitterState()
   &SpawnBurst_Instantaneous(Use Spawn Probability=false, UseLoopCountLimit=false)
  ]
  particleSpawnScript [
   &InitializeParticle(
    Color=<rgba 0.5 0.2275 0 1>
    Lifetime=0.2
    Position Mode=::NewEnumerator0
    Sprite Size Mode=::NewEnumerator3
    Uniform Sprite Size=27.5
    Write Material Random=true
   )
   &CameraOffset(Camera Offset Amount=20)
  ]
  particleUpdateScript [
   &ParticleState()
   &ScaleSpriteSize(Uniform Curve Sprite Scale=DataInterface() {propertyValues {Curve {keys [
        {arriveTangent=2.3921, interpMode=RCIM_Cubic, leaveTangent=2.3921, tangentMode=RCTM_User, time=0, value=0.25}
        {interpMode=RCIM_Cubic, tangentMode=RCTM_User, time=0.7, value=1.5}
       ]}}})
   /SetVariables(
    Particles.FlareColor=&MultiplyLinearColor(Linear Color=$Particles.Color, Scale Factor=&MakeVector4_FromFloat(Float=2))
    Particles.FlareSize:Vector2f=&Multiply_Vector2D(
     A=$Particles.SpriteSize
     B=&MakeVector2D(
      X=&FloatFromCurve(
       FloatCurve=DataInterface() {propertyValues {Curve {keys [
           {interpMode=RCIM_Cubic, tangentMode=RCTM_User, time=0, value=1}
           {arriveTangent=-0.8386, interpMode=RCIM_Cubic, leaveTangent=-0.8386, tangentMode=RCTM_User, time=1, value=0}
          ]}}}
       Scale Curve=2
      )
      Y=&FloatFromCurve(FloatCurve=DataInterface() {propertyValues {Curve {keys [<at 0 1>,<at 1 0>]}}})
     )
    )
   ) {isSetParameters=true}
  ]
  rendererClasses [&NiagaraSpriteRendererProperties]
  renderers [
   {rendererClass=&NiagaraSpriteRendererProperties}
   {rendererClass=&NiagaraSpriteRendererProperties, rendererIndex=1}
  ]
 }
 Volume {
  emitterUpdateScript [
   &EmitterState()
   &SpawnBurst_Instantaneous(Use Spawn Probability=false, UseLoopCountLimit=false)
  ]
  particleSpawnScript [
   &InitializeParticle(
    Color=<rgba 1 0.1057 0 1>
    Lifetime=0.5
    Mesh Scale=<0.6 0.6 0.6>
    Mesh Scale Mode=::NewEnumerator4
    UsePositionOffset=false
    Write Material Random=true
   )
   &CameraOffset(Camera Offset Amount=-5)
  ]
  particleUpdateScript [
   &ParticleState()
   &ScaleColor(
    Scale Alpha=&FloatFromCurve(FloatCurve=DataInterface() {propertyValues {Curve {keys [
         {arriveTangent=9.0725, interpMode=RCIM_Cubic, leaveTangent=9.0725, tangentMode=RCTM_User, time=0, value=0}
         {arriveTangent=-2.5249, interpMode=RCIM_Cubic, leaveTangent=-2.5249, tangentMode=RCTM_User, time=0.1408, value=0.9847}
         {arriveTangent=-0.0405, interpMode=RCIM_Cubic, leaveTangent=-0.0405, tangentMode=RCTM_User, time=1, value=0}
        ]}}})
    Scale RGB=&MakeVectorFromLinearColorRGB(Color=<rgba 1 0 0.0255 1>)
    ScaleA=true
    ScaleRGB=true
   )
  ]
  rendererClasses [&NiagaraMeshRendererProperties]
  renderers [{rendererClass=&NiagaraMeshRendererProperties, rendererIndex=0}]
 }
}
summary {
 systemName=NS_Emote_FearlessFlight_CloseFlash
 emitters [
  {emitterName=GlowFlash, rendererClasses [&NiagaraSpriteRendererProperties]}
  {emitterName=Volume, rendererClasses [&NiagaraMeshRendererProperties]}
 ]
 userVariables [
  User.Offset {type=NiagaraPosition, description="", defaultValue {}}
  User.Offset {type=Vector3f, description="", defaultValue {struct=Vector3f, value=<0 0 0>}}
 ]
}
