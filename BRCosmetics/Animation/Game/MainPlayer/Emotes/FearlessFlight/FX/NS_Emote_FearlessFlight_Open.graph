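# Niagara system for the opening of the Fearless Flight emote: a timed, glowing hexagon effect with glitch shapes and fragments.
# Burst emitters layer hex sprites, expanding hexes, parallax lines with a light, volume glow and fragments, while rate-spawned hex lines and jittering shapes add detail.
# Floor-quantized rotation gives a stepped spin, and time-of-day scalars dim some layers.
# User.Offset positions three emitters, and User.HexLinesScale sizes the Hexes and ParallaxLines sprites.
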NiagaraSystem /BRCosmetics/Animation/Game/MainPlayer/Emotes/FearlessFlight/FX/NS_Emote_FearlessFlight_Open
dependencies {
 usedDataInterfaces [
  &NiagaraDataInterfaceArrayFloat
  &NiagaraDataInterfaceCurve
  &NiagaraDataInterfaceArrayColor
  &NiagaraDataInterfaceColorCurve
  &NiagaraDataInterfaceArrayFloat3
 ]
 usedDynamicInputs []
 usedModules [&SystemState]
}
emitters {
 Fragments {
  emitterUpdateScript [
   &EmitterState(Life Cycle Mode=::NewEnumerator1, Loop Behavior=::NewEnumerator1, Loop Duration=2, UseLoopDelay=false)
   &SpawnBurst_Instantaneous(Use Spawn Probability=false, UseLoopCountLimit=false)
  ]
  particleSpawnScript [
   &InitializeParticle(
    Color Mode=::NewEnumerator0
    Lifetime=2
    Mesh Scale=<0.8 0.8 0.25>
    Mesh Scale Mode=::NewEnumerator4
    Position Offset=<0 -20 0>
    UsePositionOffset=true
    Write Material Random=true
   )
   &InitialMeshOrientation(Mesh Axis To Orient=<0 -1 0>, Mesh Orientation Mode=::NewEnumerator2, Orientation Vector=<0 0 1>, Use Rotation Vector=false)
  ]
  particleUpdateScript [
   &ParticleState()
   &DynamicMaterialParameters(
    Index 0 Param 1=&FloatFromCurve(FloatCurve=DataInterface() {propertyValues {Curve {keys [<at 0.5 1 RCIM_Cubic>,<at 1 0 RCIM_Cubic>]}}})
    Index 0 Param 2=&FloatFromCurve(FloatCurve=DataInterface() {propertyValues {Curve {keys [<at 0.5 1 RCIM_Cubic>,<at 1 0 RCIM_Cubic>]}}})
    Index 0 Param 3=&FloatFromCurve(FloatCurve=DataInterface() {propertyValues {Curve {keys [
         {arriveTangent=3.2202, interpMode=RCIM_Cubic, leaveTangent=3.2202, tangentMode=RCTM_User, time=0.1, value=0}
         {arriveTangent=1.1111, interpMode=RCIM_Cubic, leaveTangent=1.1111, time=0.5, value=0.75}
         <at 1 1 RCIM_Cubic>
        ]}}})
    Index 0 Param 4=&FloatFromCurve(FloatCurve=DataInterface() {propertyValues {Curve {keys [{arriveTangent=4.7912, interpMode=RCIM_Cubic, leaveTangent=4.7912, tangentMode=RCTM_User, time=0, value=0}, <at 0.5 1 RCIM_Cubic>]}}})
    Module.Param0WriteEnabled=true
    Module.Param1WriteEnabled=true
    Module.Param2WriteEnabled=true
    Module.Param3WriteEnabled=true
   )
   &ScaleColor(
    Scale Alpha=&FloatFromCurve(FloatCurve=DataInterface() {propertyValues {Curve {keys [<at 0.75 1>,<at 1 0>]}}})
    ScaleA=true
    ScaleRGB=true
   )
  ]
  rendererClasses [&NiagaraMeshRendererProperties]
  renderers [{rendererClass=&NiagaraMeshRendererProperties, rendererIndex=0}]
 }
 GlitchShapes {
  emitterUpdateScript [
   &EmitterState(Life Cycle Mode=::NewEnumerator1, Loop Behavior=::NewEnumerator1, Loop Delay=0.25, Loop Duration=4, UseLoopDelay=true)
   &SpawnRate(SpawnRate=20, Use Spawn Probability=false)
  ]
  particleSpawnScript [
   &InitializeParticle(
    Color=<rgba 5 1.8092 0 1>
    Lifetime Max=0.3
    Lifetime Min=0.2
    Lifetime Mode=::NewEnumerator1
    Mesh Scale=&Multiply_VectorByFloat(
     Vector=&RandomRangeVector(Minimum=<0.1 0.1 0.1>, Maximum=<0.15 0.15 0.15>)
     Float=&SelectFloatFromArray(Direct Array Index=0, Float Selection Array=DataInterface() {propertyValues {FloatData [0.5,0.5,0.8,0.8,0.77,0.77]}})
    )
    Mesh Scale Mode=::NewEnumerator4
    Position Offset:NiagaraPosition=$User.Offset
    UsePositionOffset=true
    Write Material Random=true
   )
   &InitialMeshOrientation(Use Rotation Vector=false)
   &ShapeLocation(
    Cylinder Height=0
    Cylinder Radius=20
    MaskSpawnGroup=false
    Offset Mode=::NewEnumerator0
    Override Seed=false
    Rotation Axis=<0 1 0>
    Rotation Mode=::NewEnumerator1
    Shape Primitive=::NewEnumerator1
    Surface Only=true
    Use Endcaps In Surface Only Mode=false
    UseOwnerScale=true
   )
   &CameraOffset(Camera Offset Amount=8)
  ]
  particleUpdateScript [
   &ParticleState()
   &Drag(Drag=10, Use Linear Drag=true, Use Rotational Drag=false)
   /SetVariables(Particles.Jitter:NiagaraFloat=&ClampFloat(
     Float=&Waveform(
      Global Clamp Min=-1
      "Waveform [1]"=::NewEnumerator12
      "[1] Frequency"=&RandomRangeFloat(Minimum=10, Maximum=20)
      "[1] Amplitude Min/Max"=<25 25>
     )
     Min=-0.25
     Max=0.25
    )) {isSetParameters=true}
   &DynamicMaterialParameters(
    Index 0 Param 1=&Add_Float(A=$Particles.Jitter, B=&RandomRangeFloat(Minimum=0, Maximum=0.1))
    Index 0 Param 2=&Subtract_Float(A=&RandomRangeFloat(Minimum=0, Maximum=0.1), B=$Particles.Jitter)
    Index 0 Param 3=$Particles.Jitter
    Module.Param0WriteEnabled=true
    Module.Param1WriteEnabled=true
    Module.Param2WriteEnabled=true
    Module.Param3WriteEnabled=false
   )
   &ScaleColor(
    Scale Alpha=&Add_Float(A=2, B=&ClampFloat(Float=$Particles.Jitter, Min=0.8))
    ScaleA=true
    ScaleRGB=true
   )
   &UpdateMeshOrientation(Rotation Rate=0.2, Rotation Vector=&RandomRangeVector(Minimum=<0 0 0>, Maximum=<1 1 1>))
   &SolveForcesAndVelocity(Clamp Velocity=false, Limit Acceleration=false)
  ]
  rendererClasses [&NiagaraMeshRendererProperties]
  renderers [{rendererClass=&NiagaraMeshRendererProperties, rendererIndex=0}]
 }
 HexLines {
  emitterUpdateScript [
   &EmitterState(Life Cycle Mode=::NewEnumerator1, Loop Behavior=::NewEnumerator1, Loop Duration=3, UseLoopDelay=false)
   &SpawnRate(SpawnRate=10, Use Spawn Probability=false)
  ]
  particleSpawnScript [
   &InitializeParticle(
    Color=&MultiplyLinearColor(
     Linear Color=&SelectLinearColorFromArray(
      Array Sampling Mode=::NewEnumerator0
      Direct Array Index=?
      Color Selection Array=DataInterface() {propertyValues {ColorData [<rgba 5 1.7 0 1>, <rgba 5 0.6636 0 1>]}}
      Random Seed=0
     )
     Scale Factor=&MakeVector4_FromFloat(Float=2)
    )
    Lifetime Max=1
    Lifetime Min=0.5
    Lifetime Mode=::NewEnumerator1
    Mesh Scale Mode=::NewEnumerator1
    Mesh Uniform Scale Max=26
    Mesh Uniform Scale Min=25
    Position Offset=&Add_Vector(
     A=Linked() {linkedVariable {name=User.Offset, type=NiagaraPosition}}
     B=&MakeVector(X=0, Y=&RandomRangeFloat(Minimum=-2.5, Maximum=2.5), Z=0)
    )
    Sprite Rotation Angle=&SelectFloatFromArray(
     Direct Array Index=&ReturnExecIndex()
     Float Selection Array=DataInterface() {propertyValues {FloatData [0.01,0,-0.01]}}
    )
    Sprite Rotation Mode=::NewEnumerator1
    Sprite Size Mode=::NewEnumerator3
    Uniform Sprite Size=43
    UsePositionOffset=true
    Write Material Random=true
   )
   &DynamicMaterialParameters(
    Index 0 Param 1=&RandomRangeFloat(Minimum=0, Maximum=10)
    Index 0 Param 2=&RandomRangeFloat(Minimum=0, Maximum=1)
    Module.Param0WriteEnabled=true
    Module.Param1WriteEnabled=true
    Module.Param2WriteEnabled=true
    Module.Param3WriteEnabled=true
   )
  ]
  particleUpdateScript [
   &ParticleState()
   &ScaleColor(
    Scale Alpha=&FloatFromCurve(
     FloatCurve=DataInterface() {propertyValues {Curve {keys [
         {interpMode=RCIM_Cubic, tangentMode=RCTM_User, time=0, value=0}
         {interpMode=RCIM_Cubic, tangentMode=RCTM_User, time=0.5, value=1}
         {tangentMode=RCTM_User, time=1, value=0}
        ]}}}
     Scale Curve=&FNTimeOfDayScalar(Night Scale=0.7, Morning Scale=0.9, Evening Scale=0.9)
    )
    ScaleA=true
    ScaleRGB=true
   )
  ]
  rendererClasses [&NiagaraMeshRendererProperties]
  renderers [{rendererClass=&NiagaraMeshRendererProperties, rendererIndex=0}]
 }
 Hex_Expand {
  emitterUpdateScript [
   &EmitterState(Life Cycle Mode=::NewEnumerator1, Loop Behavior=::NewEnumerator1, Loop Delay=0.25, Loop Duration=2, UseLoopDelay=true)
   &SpawnBurst_Instantaneous(Spawn Count=2, Use Spawn Probability=false, UseLoopCountLimit=false)
  ]
  particleSpawnScript [
   &InitializeParticle(
    Color Mode=::NewEnumerator0
    Mesh Scale Mode=::NewEnumerator3
    Mesh Uniform Scale=&Multiply_Float(
     A=6
     B=&FloatFromCurve(
      FloatCurve=DataInterface() {propertyValues {Curve {keys [{tangentMode=RCTM_User, time=0, value=0.6}, {tangentMode=RCTM_User, time=1, value=1}]}}}
      CurveIndex=&ReturnNormalizedExecIndex()
     )
    )
    Sprite Rotation Mode=::NewEnumerator1
    Sprite Size Mode=::NewEnumerator3
    Uniform Sprite Size=&Multiply_Float(
     A=140
     B=&FloatFromCurve(
      FloatCurve=DataInterface() {propertyValues {Curve {keys [{tangentMode=RCTM_User, time=0, value=0.6}, {tangentMode=RCTM_User, time=1, value=1}]}}}
      CurveIndex=&ReturnNormalizedExecIndex()
     )
    )
    UsePositionOffset=false
    Write Material Random=true
   )
   &SpriteFacingAndAlignment(Sprite Facing=<0 1 0>, WriteAlignment=false, WriteFacing=true)
  ]
  particleUpdateScript [
   &ParticleState()
   &ScaleMeshSize(Scale Factor=&Multiply_VectorByFloat(
     Vector=<1 1 1>
     Float=&FloatFromCurve(FloatCurve=DataInterface() {propertyValues {Curve {keys [
          {arriveTangent=4.1785, interpMode=RCIM_Cubic, leaveTangent=4.1785, tangentMode=RCTM_User, time=0, value=0}
          {arriveTangent=1.1179, interpMode=RCIM_Cubic, leaveTangent=1.1179, tangentMode=RCTM_User, time=0.2042, value=0.6228}
          {arriveTangent=0.0224, interpMode=RCIM_Cubic, leaveTangent=0.0224, tangentMode=RCTM_User, time=1, value=1}
         ]}}})
    ))
   /SetVariables(Particles.MeshOrientation=&MakeQuatFromAxisAngle(
     Axis=&MakeVector(X=0, Y=1, Z=0)
     AngleInDegrees=&AngleConversion(
      Angle:NiagaraFloat=&Multiply_Float(
       A=&FloatFromCurve(
        FloatCurve=DataInterface() {propertyValues {Curve {keys [{arriveTangent=1.0177, interpMode=RCIM_Cubic, leaveTangent=1.0177, tangentMode=RCTM_User, time=0, value=0.5}, <at 1 1 RCIM_Cubic>]}}}
        Scale Curve=&FloatFromCurve(
         FloatCurve=DataInterface() {propertyValues {Curve {keys [{tangentMode=RCTM_User, time=0, value=0.675}, {tangentMode=RCTM_User, time=1, value=1}]}}}
         CurveIndex=&ReturnNormalizedExecIndex()
        )
       )
       B=1
      )
      Angle Input:ENiagara_AngleInput=::NewEnumerator1
      Angle Output:ENiagara_AngleInput=::NewEnumerator0
     )
     NormalizeAxis=true
    )) {enabled=false, isSetParameters=true}
   /SetVariables(Particles.MeshOrientation=&MakeQuatFromAxisAngle(
     Axis=&MakeVector(X=0, Y=1, Z=0)
     AngleInDegrees=&Multiply_Float(
      A=&Divide_Float(A=/Niagara/DynamicInputs/Math/Floor_Float(Float:NiagaraFloat=&Multiply_Float(A=$Engine.RealTime, B=12)), B=12)
      B=&Multiply_Float(
       A=&FloatFromCurve(
        FloatCurve=DataInterface() {propertyValues {Curve {keys [{tangentMode=RCTM_User, time=0, value=0.7}, {tangentMode=RCTM_User, time=1, value=1}]}}}
        CurveIndex=&ReturnNormalizedExecIndex()
       )
       B=125
      )
     )
     NormalizeAxis=true
    )) {isSetParameters=true}
   &DynamicMaterialParameters(
    Index 0 Param 1=&FloatFromCurve(FloatCurve=DataInterface() {propertyValues {Curve {keys [{tangentMode=RCTM_User, time=0, value=0}, {tangentMode=RCTM_User, time=1, value=1}]}}})
    Module.Param0WriteEnabled=true
    Module.Param1WriteEnabled=true
    Module.Param2WriteEnabled=true
    Module.Param3WriteEnabled=true
   )
   &ScaleColor(
    Scale Alpha=&FloatFromCurve(FloatCurve=DataInterface() {propertyValues {Curve {keys [
         {arriveTangent=9.0725, interpMode=RCIM_Cubic, leaveTangent=9.0725, tangentMode=RCTM_User, time=0, value=0}
         {arriveTangent=-2.5249, interpMode=RCIM_Cubic, leaveTangent=-2.5249, tangentMode=RCTM_User, time=0.1408, value=0.9847}
         {arriveTangent=-0.0405, interpMode=RCIM_Cubic, leaveTangent=-0.0405, tangentMode=RCTM_User, time=1, value=0}
        ]}}})
    Scale RGB=&MakeVectorFromLinearColorRGB(Color=&ColorFromCurve(ColorCurve=DataInterface() {propertyValues {
        AlphaCurve {keys [<at 0 1 RCIM_Cubic>]}
        BlueCurve {keys [<at 0 2.3612 RCIM_Cubic>,<at 0.0538 0>,<at 0.1319 0>,<at 0.2072 5>,<at 0.6378 1 RCIM_Cubic>]}
        GreenCurve {keys [<at 0 0 RCIM_Cubic>,<at 0.0538 0.0182>,<at 0.1319 0.0182>,<at 0.2072 0.344>,<at 0.6378 0.1714 RCIM_Cubic>]}
        RedCurve {keys [<at 0 5 RCIM_Cubic>,<at 0.0538 2>,<at 0.1319 2>,<at 0.2072 0>,<at 0.6378 0 RCIM_Cubic>]}
       }}))
    ScaleA=true
    ScaleRGB=true
   )
  ]
  rendererClasses [&NiagaraMeshRendererProperties]
  renderers [{rendererClass=&NiagaraMeshRendererProperties, rendererIndex=0}]
 }
 Hexes {
  emitterUpdateScript [
   &EmitterState()
   &SpawnBurst_Instantaneous(Spawn Count=3, Use Spawn Probability=false, UseLoopCountLimit=false)
  ]
  particleSpawnScript [
   &InitializeParticle(
    Color=<rgba 0.7468 0 2 1>
    Lifetime=4
    Position Offset=&Add_Vector(
     A=Linked() {linkedVariable {name=User.Offset, type=NiagaraPosition}}
     B=&MakeVector(
      X=0
      Y=&FloatFromCurve(
       FloatCurve=DataInterface() {propertyValues {Curve {keys [<at 0 -1>,<at 1 1>]}}}
       CurveIndex=&ReturnNormalizedExecIndex()
       Scale Curve=2
      )
      Z=0
     )
    )
    Sprite Rotation Mode=::NewEnumerator1
    Sprite Size Mode=::NewEnumerator3
    Uniform Sprite Size=62
    UsePositionOffset=true
    Write Material Random=true
   )
   &SpriteFacingAndAlignment(Sprite Facing=<0 1 0>, WriteAlignment=false, WriteFacing=true)
  ]
  particleUpdateScript [
   &ParticleState()
   &ScaleSpriteSize(Scale Sprite Size Mode=::NewEnumerator0, Uniform Scale Factor=$User.HexLinesScale)
   /SetVariables(Particles.SpriteRotation:NiagaraFloat=&Waveform(
     Global Input Position=&FloatFromCurve(
      FloatCurve=DataInterface() {propertyValues {Curve {keys [<at 0 1>,<at 1 0>]}}}
      CurveIndex=&ReturnNormalizedExecIndex()
      Scale Curve=0.33
     )
     Global Amplitude Scale=50
     "Waveform [1]"=::NewEnumerator0
     "[1] Phase Scale"=0.25
     "[1] Frequency"=1
     "[1] Amplitude Min/Max"=<0 1>
     "[1] Scale"=0.5
     "[1] Bias"=&FloatFromCurve(FloatCurve=DataInterface() {propertyValues {Curve {keys [
          <at 0 -1 RCIM_Cubic>
          {arriveTangent=-0.0147, leaveTangent=-0.0147, time=0.2, value=1.31}
          {arriveTangent=-2.8875, interpMode=RCIM_Cubic, leaveTangent=-2.8875, time=0.8, value=1.31}
          <at 1 -1 RCIM_Cubic>
         ]}}})
    )) {isSetParameters=true}
   /SetVariables(Particles.SpriteRotation:NiagaraFloat=&Multiply_Float(
     A=&Floor_Float(Float:NiagaraFloat=&Divide_Float(A=$Particles.SpriteRotation, B=2))
     B=2
    )) {isSetParameters=true}
   &ScaleColor(
    Scale RGB=&Multiply_VectorByFloat(Vector=<1 1 1>, Float=&FNTimeOfDayScalar(Night Scale=0.6, Morning Scale=0.9, Evening Scale=0.9))
    ScaleA=true
    ScaleRGB=true
   )
  ]
  rendererClasses [&NiagaraSpriteRendererProperties]
  renderers [{rendererClass=&NiagaraSpriteRendererProperties}]
 }
 ParallaxLines {
  emitterUpdateScript [
   &EmitterState(Life Cycle Mode=::NewEnumerator1, Loop Behavior=::NewEnumerator1, Loop Delay=0.1, Loop Duration=3, UseLoopDelay=false)
   &SpawnBurst_Instantaneous(Spawn Count=2, Use Spawn Probability=false, UseLoopCountLimit=false)
  ]
  particleSpawnScript [
   &InitializeParticle(
    Color=<rgba 22 4.5742 0 2>
    Lifetime=3.825
    Position Offset=&SelectVectorFromArray(
     Direct Array Index=&ReturnExecIndex()
     Vector Selection Array=DataInterface() {propertyValues {FloatData [<0 0 0>, <0 -0.25 0>]}}
    )
    Sprite Size Mode=::NewEnumerator3
    Uniform Sprite Size=44
    UsePositionOffset=true
    Write Material Random=true
   )
   &SpriteFacingAndAlignment(
    Sprite Alignment=<0 0 1>
    Sprite Facing=&SelectVectorFromArray(
     Direct Array Index=&ReturnExecIndex()
     Vector Selection Array=DataInterface() {propertyValues {FloatData [<0 1 0>, <0 -1 0>]}}
    )
    WriteAlignment=true
    WriteFacing=true
   )
  ]
  particleUpdateScript [
   &ParticleState()
   &ScaleSpriteSize(Scale Sprite Size Mode=::NewEnumerator0, Uniform Scale Factor=$User.HexLinesScale)
   &ScaleColor(
    Scale Alpha=&FloatFromCurve(
     FloatCurve=DataInterface() {propertyValues {Curve {keys [
         {interpMode=RCIM_Cubic, tangentMode=RCTM_User, time=0, value=0}
         <at 0.1 1>
         {interpMode=RCIM_Cubic, tangentMode=RCTM_User, time=0.75, value=1}
         {tangentMode=RCTM_User, time=0.9, value=0}
        ]}}}
     Scale Curve=&FloatFromCurve(FloatCurve=DataInterface() {propertyValues {Curve {keys [
          {interpMode=RCIM_Cubic, tangentMode=RCTM_User, time=0, value=0}
          <at 0.1 1>
          {interpMode=RCIM_Cubic, tangentMode=RCTM_User, time=0.75, value=1}
          {tangentMode=RCTM_User, time=0.9, value=0}
         ]}}})
    )
    ScaleA=true
    ScaleRGB=true
   )
   &FN_Light_Attributes(
    Color=$Particles.Color
    Create Color Attribute=true
    Create Exponent Attribute=true
    Create Position Attribute=true
    Create Radius Attribute=true
    Exponent=3
    Position:NiagaraPosition=$Particles.Position
    Radius=50
    Scale Color=&Waveform(
     Global Amplitude Scale=0.1
     "Waveform [1]"=::NewEnumerator0
     "[1] Frequency"=0.5
     "[1] Amplitude Min/Max"=<0 1>
     "[1] Scale"=0.6
     "[1] Bias"=0.5
    )
   )
  ]
  rendererClasses [&NiagaraSpriteRendererProperties, &NiagaraLightRendererProperties]
  renderers [
   {rendererClass=&NiagaraSpriteRendererProperties}
   {rendererClass=&NiagaraLightRendererProperties, rendererIndex=1}
  ]
 }
 Volume {
  emitterUpdateScript [
   &EmitterState(Life Cycle Mode=::NewEnumerator1, Loop Behavior=::NewEnumerator1, Loop Delay=0.25, Loop Duration=3, UseLoopDelay=true)
   &SpawnBurst_Instantaneous(Use Spawn Probability=false, UseLoopCountLimit=false)
  ]
  particleSpawnScript [
   &InitializeParticle(
    Color Mode=::NewEnumerator0
    Lifetime=3.5
    Mesh Scale=<0.8 0.3 0.8>
    Mesh Scale Mode=::NewEnumerator4
    Position Offset=<0 5 0>
    UsePositionOffset=true
    Write Material Random=true
   )
   &CameraOffset(Camera Offset Amount=-5)
  ]
  particleUpdateScript [
   &ParticleState()
   &ScaleColor(
    Scale Alpha=&FloatFromCurve(FloatCurve=DataInterface() {propertyValues {Curve {keys [
         {interpMode=RCIM_Cubic, tangentMode=RCTM_User, time=0, value=0}
         <at 0.1 1>
         {interpMode=RCIM_Cubic, tangentMode=RCTM_User, time=0.8, value=1}
         {tangentMode=RCTM_User, time=1, value=0}
        ]}}})
    Scale RGB=&MakeVectorFromLinearColorRGB(Color=<rgba 1 0 0.6195 1>)
    ScaleA=true
    ScaleRGB=true
   )
   /SetVariables(
    Particles.SoftGlowColor:LinearColor=&ColorFromCurve(ColorCurve=DataInterface() {propertyValues {
       AlphaCurve {keys [<at 0 0>, <at 0.0269 1>, {interpMode=RCIM_Cubic, leaveTangent=-2.2022, time=0.2246, value=1}, <at 0.4468 0>]}
       BlueCurve {keys [<at 0 0 RCIM_Cubic>,<at 0.1172 0.9588>,<at 0.2075 1>,<at 0.4468 1 RCIM_Cubic>]}
       GreenCurve {keys [<at 0 0.0343 RCIM_Cubic>,<at 0.1172 0>,<at 0.2075 0.3839>,<at 0.4468 0.1115 RCIM_Cubic>]}
       RedCurve {keys [<at 0 1 RCIM_Cubic>,<at 0.1172 1>,<at 0.2075 0>,<at 0.4468 0 RCIM_Cubic>]}
      }})
    Particles.SoftGlowScale:Vector3f=&Multiply_VectorByFloat(Vector=$Particles.Scale, Float=2)
   ) {isSetParameters=true}
  ]
  rendererClasses [&NiagaraMeshRendererProperties]
  renderers [
   {rendererClass=&NiagaraMeshRendererProperties, rendererIndex=0}
   {rendererClass=&NiagaraMeshRendererProperties, rendererIndex=1}
  ]
 }
}
summary {
 systemName=NS_Emote_FearlessFlight_Open
 emitters [
  {emitterName=GlitchShapes, rendererClasses [&NiagaraMeshRendererProperties]}
  {emitterName=Hexes, rendererClasses [&NiagaraSpriteRendererProperties]}
  {emitterName=HexLines, rendererClasses [&NiagaraMeshRendererProperties]}
  {emitterName=Volume, rendererClasses [&NiagaraMeshRendererProperties]}
  {emitterName=ParallaxLines, rendererClasses [&NiagaraSpriteRendererProperties, &NiagaraLightRendererProperties]}
  {emitterName=Hex_Expand, rendererClasses [&NiagaraMeshRendererProperties]}
  {emitterName=Fragments, rendererClasses [&NiagaraMeshRendererProperties]}
 ]
 userVariables [
  User.HexLinesScale {type=NiagaraFloat, defaultValue=1, description=""}
  User.Offset {type=NiagaraPosition, description="", defaultValue {}}
  User.SkeletalMesh {
   type=NiagaraDataInterfaceSkeletalMesh
   defaultValue {
    struct=NiagaraExt_VariableValue_DataInterface
    value {dataInterface=&self:NiagaraDataInterfaceSkeletalMesh_0}
   }
  }
 ]
}
